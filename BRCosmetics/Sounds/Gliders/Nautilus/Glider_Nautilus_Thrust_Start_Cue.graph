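# Sound cue for the one-shot thrust start of the Nautilus glider cosmetic.
# A Random node picks one of four equally weighted start waves, and a Modulator slightly varies its volume.
# The cue plays at reduced volume through the Fort_Gliders_Land sound class, with player glider thrust attenuation and a glider thrust concurrency set.
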SoundCue /BRCosmetics/Sounds/Gliders/Nautilus/Glider_Nautilus_Thrust_Start_Cue
root=Modulator_0
nodes [
 {id=Modulator_0, children [Random_1], properties {volumeMax=1, volumeMin=0.85}}
 {id=Random_1, children [WavePlayer_1,WavePlayer_5,WavePlayer_6,WavePlayer_7], properties {weights [1,1,1,1]}}
 {id=WavePlayer_1, properties {soundWaveAssetPtr=&/Glider_Nautilus_Thrust_Start_01}}
 {id=WavePlayer_5, properties {soundWaveAssetPtr=&/Glider_Nautilus_Thrust_Start_02}}
 {id=WavePlayer_6, properties {soundWaveAssetPtr=&/Glider_Nautilus_Thrust_Start_03}}
 {id=WavePlayer_7, properties {soundWaveAssetPtr=&/Glider_Nautilus_Thrust_Start_04}}
]
properties {
 attenuationSettings=&Fort_Player_Glider_Thrust_Athena_Att
 maxDistance=12500
 soundClassObject=&Fort_Gliders_Land
 volumeMultiplier=0.5
 concurrencySet [&Athena_Glider_Thrust_Conc]
}
